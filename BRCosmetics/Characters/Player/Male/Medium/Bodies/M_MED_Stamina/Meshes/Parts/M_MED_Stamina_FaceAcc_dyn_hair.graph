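# Physics asset for the hair on the Stamina male medium player cosmetic's face accessory, defining bodies and joints for secondary hair motion.
# Five top and three back locks (one left, two right) each pair an offset capsule with a small capsule on a matching MSTR bone.
# One constraint joins each pair, allowing wider swing on top than in back and very little twist, with no joints between locks.
# Back-hair shapes reuse the top_hair_05 capsule names.
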PhysicsAsset /BRCosmetics/Characters/Player/Male/Medium/Bodies/M_MED_Stamina/Meshes/Parts/M_MED_Stamina_FaceAcc_dyn_hair
bodies {
 dyn_bk_hair_l_01 [{shapeType=Capsule, length=5, radius=5, shapeName=dyn_top_hair_05_capsule, center=<8 0 0>, rotation=<rot 90 0 0>}]
 dyn_bk_hair_l_01_MSTR [{shapeType=Capsule, length=1, radius=1, shapeName=dyn_top_hair_05_MSTR_capsule, rotation=<rot 90 0 0>}]
 dyn_bk_hair_r_01 [{shapeType=Capsule, length=5, radius=5, shapeName=dyn_top_hair_05_capsule, center=<8 0 0>, rotation=<rot 90 0 0>}]
 dyn_bk_hair_r_01_MSTR [{shapeType=Capsule, length=1, radius=1, shapeName=dyn_top_hair_05_MSTR_capsule, rotation=<rot 90 0 0>}]
 dyn_bk_hair_r_02 [{shapeType=Capsule, length=5, radius=5, shapeName=dyn_top_hair_05_capsule, center=<10 0 0>, rotation=<rot 90 0 0>}]
 dyn_bk_hair_r_02_MSTR [{shapeType=Capsule, length=1, radius=1, shapeName=dyn_top_hair_05_MSTR_capsule, rotation=<rot 90 0 0>}]
 dyn_top_hair_01 [{shapeType=Capsule, length=5, radius=5, shapeName=dyn_top_hair_01_capsule, center=<10 0 0>, rotation=<rot 90 0 0>}]
 dyn_top_hair_01_MSTR [{shapeType=Capsule, length=1, radius=1, shapeName=dyn_top_hair_01_MSTR_capsule, rotation=<rot 90 0 0>}]
 dyn_top_hair_02 [{shapeType=Capsule, length=5, radius=5, shapeName=dyn_top_hair_02_capsule, center=<12 0 0>, rotation=<rot 90 0 0>}]
 dyn_top_hair_02_MSTR [{shapeType=Capsule, length=1, radius=1, shapeName=dyn_top_hair_02_MSTR_capsule, rotation=<rot 90 0 0>}]
 dyn_top_hair_03 [{shapeType=Capsule, length=5, radius=5, shapeName=dyn_top_hair_03_capsule, center=<11 0 0>, rotation=<rot 90 0 0>}]
 dyn_top_hair_03_MSTR [{shapeType=Capsule, length=1, radius=1, shapeName=dyn_top_hair_03_MSTR_capsule, rotation=<rot 90 0 0>}]
 dyn_top_hair_04 [{shapeType=Capsule, length=5, radius=5, shapeName=dyn_top_hair_04_capsule, center=<8 0 0>, rotation=<rot 90 0 0>}]
 dyn_top_hair_04_MSTR [{shapeType=Capsule, length=1, radius=1, shapeName=dyn_top_hair_04_MSTR_capsule, rotation=<rot 90 0 0>}]
 dyn_top_hair_05 [{shapeType=Capsule, length=5, radius=5, shapeName=dyn_top_hair_05_capsule, center=<10 0 0>, rotation=<rot 90 0 0>}]
 dyn_top_hair_05_MSTR [{shapeType=Capsule, length=1, radius=1, shapeName=dyn_top_hair_05_MSTR_capsule, rotation=<rot 90 0 0>}]
}
constraints [
 {bone1Name=dyn_top_hair_01, bone2Name=dyn_top_hair_01_MSTR, swing1LimitDegrees=30, swing2LimitDegrees=30, twistLimitDegrees=2}
 {bone1Name=dyn_top_hair_02, bone2Name=dyn_top_hair_02_MSTR, swing1LimitDegrees=30, swing2LimitDegrees=30, twistLimitDegrees=2}
 {bone1Name=dyn_top_hair_03, bone2Name=dyn_top_hair_03_MSTR, swing1LimitDegrees=30, swing2LimitDegrees=30, twistLimitDegrees=2}
 {bone1Name=dyn_top_hair_04, bone2Name=dyn_top_hair_04_MSTR, swing1LimitDegrees=30, swing2LimitDegrees=30, twistLimitDegrees=2}
 {bone1Name=dyn_top_hair_05, bone2Name=dyn_top_hair_05_MSTR, swing1LimitDegrees=30, swing2LimitDegrees=30, twistLimitDegrees=2}
 {bone1Name=dyn_bk_hair_l_01, bone2Name=dyn_bk_hair_l_01_MSTR, swing1LimitDegrees=15, swing2LimitDegrees=15, twistLimitDegrees=2}
 {bone1Name=dyn_bk_hair_r_01, bone2Name=dyn_bk_hair_r_01_MSTR, swing1LimitDegrees=15, swing2LimitDegrees=15, twistLimitDegrees=2}
 {bone1Name=dyn_bk_hair_r_02, bone2Name=dyn_bk_hair_r_02_MSTR, swing1LimitDegrees=15, swing2LimitDegrees=15, twistLimitDegrees=2}
]
